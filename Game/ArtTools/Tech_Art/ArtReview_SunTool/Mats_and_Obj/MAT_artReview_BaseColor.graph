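# Post-process material in the ArtReview SunTool art tools, built to show the scene's GBuffer base color for art review.
# A SceneTexture node samples base color and feeds a Power node.
# An Exp scalar parameter drives the exponent, with a default close to 1/2.2 that acts as a linear-to-display gamma correction.
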Material /Game/ArtTools/Tech_Art/ArtReview_SunTool/Mats_and_Obj/MAT_artReview_BaseColor
graph {
 properties {materialDomain=MD_PostProcess, pixelDepthOffsetMode=PDOM_Legacy, usedShadingModels=Default Lit}
 nodes [Power(Base=SceneTexture() {sceneTextureId=PPI_BaseColor}.Color, Exp=ScalarParameter() {defaultValue=0.45, parameterName=Exp, sliderMax=1}) {constExponent=0.45}]
}
