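# Unlit material in the ArtReview SunTool art folder, apparently meant for inspecting mesh vertex colors during art review.
# The graph holds a single Vertex Color node, with no parameters, textures or math, and shows no link from it to a material output.
# Usage flags enable it on skeletal meshes and with mesh deformers.
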Material /Game/ArtTools/Tech_Art/ArtReview_SunTool/Mats_and_Obj/MAT_artReview_VertexColor
graph {
 properties {bUsedWithMeshDeformer=true, bUsedWithSkeletalMesh=true, pixelDepthOffsetMode=PDOM_Legacy, shadingModel=MSM_Unlit, usedShadingModels=Unlit}
 nodes [VertexColor()]
}
